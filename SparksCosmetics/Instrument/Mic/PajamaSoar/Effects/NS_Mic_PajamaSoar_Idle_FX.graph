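# Niagara system for the idle glow on the PajamaSoar microphone cosmetic, built from one combined emitter with two sprite and two mesh renderers.
# A burst of four particles is partitioned by visibility tag, with per-partition arrays setting size, color, camera offset and socket placement.
# Particles follow the skeletal mesh every update and scale with system mesh-scale values.
# Glow color and alpha respond to time of day and a front-end preview flag.
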NiagaraSystem /SparksCosmetics/Instrument/Mic/PajamaSoar/Effects/NS_Mic_PajamaSoar_Idle_FX
dependencies {
 usedDataInterfaces [
  &NiagaraDataInterfaceArrayColor
  &NiagaraDataInterfaceArrayFloat2
  &NiagaraDataInterfaceArrayFloat
  &NiagaraDataInterfaceArrayFloat3
  &NiagaraDataInterfaceCurve
  &NiagaraDataInterfaceArrayInt32
  &NiagaraDataInterfaceSkeletalMesh
  &NiagaraDataInterfaceColorCurve
 ]
 usedDynamicInputs [
  &MakeFloatFromVector
  &Add_Float
  &ColorFromCurve
 ]
 usedModules [
  &self:SystemSpawnScript.SystemScriptSource.SystemScriptGraph.NiagaraNodeAssignment_0.SetVariables_#1
  &self:SystemSpawnScript.SystemScriptSource.SystemScriptGraph.NiagaraNodeAssignment_1.SetVariables_#2
  &self:SystemSpawnScript.SystemScriptSource.SystemScriptGraph.NiagaraNodeAssignment_2.SetVariables_#3
  &SystemState
 ]
}
emitters {Mic_Glow_IdleFX_Combined {
  emitterUpdateScript [
   &EmitterState(Life Cycle Mode=::NewEnumerator1, Loop Behavior=::NewEnumerator1, Loop Duration Mode=::NewEnumerator1, UseLoopDelay=false)
   &SpawnBurst_Instantaneous(Spawn Count=4, Use Spawn Probability=true, UseLoopCountLimit=false)
   &SpawnBurst_Instantaneous(Spawn Group=1, Spawn Probability=$User.IncludeMicFX, Use Spawn Probability=true, UseLoopCountLimit=false) {enabled=false, module=SpawnBurst_Instantaneous001}
  ]
  particleSpawnScript [
   &PartitionParticles(Number Of Partitions=4)
   /SetVariables(Particles.VisibilityTag=$Particles.PartitionParticles.Partition) {isSetParameters=true}
   &InitializeParticle(
    Color=&SelectLinearColorFromArray(
     Array Sampling Mode=::NewEnumerator1
     Direct Array Index=$Particles.PartitionParticles.Partition
     Color Selection Array=DataInterface() {propertyValues {ColorData [<rgba 1 1 1 0.75>, <rgba 1 1 1 0.75>, <rgba 1 1 1 1>, <rgba 1 1 1 1>]}}
     Random Seed=?
    )
    Lifetime=$Emitter.CurrentLoopDuration
    Mesh Scale Mode=::NewEnumerator3
    Mesh Uniform Scale=&SelectFloatFromArray(Direct Array Index=$Particles.PartitionParticles.Partition, Float Selection Array=DataInterface() {propertyValues {FloatData [1,1,1.2,1]}})
    Sprite Rotation Mode=::NewEnumerator0
    Sprite Size=&SelectVector2DFromArray(
     Direct Array Index=$Particles.PartitionParticles.Partition
     Vector 2D Selection Array=DataInterface() {propertyValues {FloatData [<11.5 17.5>, <15.75 10>]}}
    )
    Sprite Size Mode=::NewEnumerator4
    Sprite UV Mode=::NewEnumerator4
    UsePositionOffset=false
    Write Material Random=true
   )
   &CameraOffset(Camera Offset Amount=&SelectFloatFromArray(Direct Array Index=$Particles.PartitionParticles.Partition, Float Selection Array=DataInterface() {propertyValues {FloatData [-3,-3,5,-2]}}))
   &InitialMeshOrientation(
    Mesh Orientation Mode=::NewEnumerator2
    Rotation=&SelectVectorFromArray(
     Direct Array Index=$Particles.PartitionParticles.Partition
     Vector Selection Array=DataInterface() {propertyValues {FloatData [<0 0 0>, <0 0 0>, <0 0 0.25>, <0 0 0>]}}
    )
    Use Rotation Vector=true
   )
  ]
  particleUpdateScript [
   &ParticleState(Kill Particles When Lifetime Has Elapsed=false, Let Infinitely Lived Particles Die When Emitter Deactivates=true)
   &ScaleSpriteSize(Scale Sprite Size Mode=::NewEnumerator0, Uniform Scale Factor=$System.MeshScaleRemap)
   &ScaleMeshSize(Scale Factor=&VectorFromFloat(Value=$System.MeshScaleRemap))
   &DynamicMaterialParameters(
    Index 0 Param 1=&Multiply_Float(A=&Multiply_Float(A=7, B=$System.MeshScaleAdjusted), B=1)
    Index 0 Param 2=&Multiply_Float(A=&Multiply_Float(A=7, B=$System.MeshScaleAdjusted), B=1)
    Index 1 Param 1=&Multiply_Float(A=&Multiply_Float(A=7.75, B=$System.MeshScaleAdjusted), B=1)
    Index 1 Param 2=&Multiply_Float(A=9, B=$System.MeshScaleAdjusted)
    Module.Index0Float0Write=true
    Module.Index0Float1Write=true
    Module.Index0Float2Write=true
    Module.Index0Float3Write=true
    Module.Index1Float0Write=true
    Module.Index1Float1Write=true
    Module.Index1Float2Write=true
    Module.Index1Float3Write=true
    Write Parameter Index 1=::NewEnumerator1
   )
   &ScaleColor(
    Scale Alpha=&Multiply_Float(
     A=&MakeCustomFloatFromBool(Bool=$User.bIsFrontEndPreview, True Float=0.5, False Float=1)
     B=&FNTimeOfDayScalar(Night Scale=1.1, Morning Scale=1.15, Day Scale=1.15, Evening Scale=1.35)
    )
    Scale RGB=&Multiply_Vector(
     A=&VectorFromFloat(Value=1.75)
     B=&MakeVectorFromLinearColorRGB(Color=$System.TOD Color)
    )
    ScaleA=true
    ScaleRGB=true
   )
   &SkeletalMeshLocation(
    Bone / Socket Index=&SelectIntFromArray(Direct Array Index=$Particles.PartitionParticles.Partition, Int Selection Array=DataInterface() {propertyValues {IntData [0,0,0,2]}})
    MaskSpawnGroup=false
    Mesh Sampling Type=::NewEnumerator5
    Sampled Position Offset=&Multiply_VectorByFloat(
     Vector=&SelectVectorFromArray(
      Direct Array Index=$Particles.PartitionParticles.Partition
      Vector Selection Array=DataInterface() {propertyValues {FloatData [<-2.6 0 0>, <-2 0 6.2>, <-1.65 0 2.88>, <-2.58 0 -0.85>]}}
     )
     Float=$System.MeshScaleRemap
    )
    Skeletal Mesh=$System.Skeletal Mesh
    Socket Sampling Mode=::NewEnumerator5
   )
   /SetVariables(Particles.SpriteAlignment=<0 1 0>, Particles.SpriteFacing1:Vector3f=<0 0 1>) {isSetParameters=true}
   /SetVariables(
    Particles.GlowMeshColor:LinearColor=&MakeLinearColorFromVectorAndFloat(
     "Vector (RGB)"=&Multiply_Vector(
      A=&Multiply_Vector(
       A=&MakeVectorFromLinearColorRGB(Color=$Particles.Initial.Color)
       B=&VectorFromFloat(Value=1)
      )
      B=&MakeVectorFromLinearColorRGB(Color=$System.TOD Color)
     )
     "Float (Alpha)"=&Multiply_Float(
      A=&MakeCustomFloatFromBool(Bool=$User.bIsFrontEndPreview, True Float=0.8, False Float=0.975)
      B=&FNTimeOfDayScalar(Night Scale=1, Morning Scale=1, Evening Scale=1)
     )
    )
    Particles.HandleGlowColor:LinearColor=&MakeLinearColorFromVectorAndFloat(
     "Vector (RGB)"=&Multiply_Vector(
      A=&Multiply_Vector(
       A=&MakeVectorFromLinearColorRGB(Color=$Particles.Initial.Color)
       B=&VectorFromFloat(Value=1.75)
      )
      B=&MakeVectorFromLinearColorRGB(Color=$System.TOD Color)
     )
     "Float (Alpha)"=&Multiply_Float(
      A=&MakeCustomFloatFromBool(Bool=$User.bIsFrontEndPreview, True Float=1, False Float=3)
      B=&FNTimeOfDayScalar(Night Scale=1.15, Morning Scale=1.15, Evening Scale=1)
     )
    )
   ) {isSetParameters=true}
  ]
  rendererClasses [&NiagaraSpriteRendererProperties, &NiagaraMeshRendererProperties]
  renderers [
   {rendererClass=&NiagaraSpriteRendererProperties}
   {rendererClass=&NiagaraSpriteRendererProperties, rendererIndex=1}
   {rendererClass=&NiagaraMeshRendererProperties, rendererIndex=2}
   {rendererClass=&NiagaraMeshRendererProperties, rendererIndex=3}
  ]
 }}
summary {
 systemName=NS_Mic_PajamaSoar_Idle_FX
 emitters [{
   emitterName=Mic_Glow_IdleFX_Combined
   rendererClasses [&NiagaraSpriteRendererProperties, &NiagaraMeshRendererProperties]
  }]
 userVariables [
  User.bIsFrontEndPreview {type=NiagaraBool}
  User.IncludeMicFX {type=NiagaraFloat, defaultValue=1, description=""}
  User.IsBackpack {type=NiagaraFloat, defaultValue=0, description=""}
 ]
}
